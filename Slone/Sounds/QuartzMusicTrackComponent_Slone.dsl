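; Quartz music track component for Slone that mixes four adaptive music layers: unaware, base, guard and boss.
; Each data tracker update calls the parent, then Mix State converts clamped tracker values, some summed, into four track parameters.
; Base combines four tracker values, including those behind guard and boss.
; Once boss rises above a small threshold, the guard layer ignores its own value and becomes the inverse of boss, crossfading between them.
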(event Custom|MIx%20State
  (Control|SetTrackParameter 0 (CallFunction|GetUnawareValue))
  (Control|SetTrackParameter 1 (CallFunction|GetBaseValue))
  (Control|SetTrackParameter 2 (Math|Float|SelectFloat (Math|Float|Lerp 1.0 0.0 (CallFunction|GetBossValue)) (CallFunction|GetGuardValue) (>= (CallFunction|GetBossValue) 0.01)))
  (Control|SetTrackParameter 3 (CallFunction|GetBossValue)))

(fn GetBaseValue
  (return (Math|Float|MapRangeClamped (+ (+ (CallFunction|GetDataTrackerParameterValue 2) (CallFunction|GetDataTrackerParameterValue 5)) (+ (CallFunction|GetDataTrackerParameterValue 1) (CallFunction|GetDataTrackerParameterValue 4))) 0.0 1.0 0.0 1.0)))

(fn GetBossValue
  (return (Math|Float|MapRangeClamped (CallFunction|GetDataTrackerParameterValue 5) 0.0 1.0 0.0 1.0)))

(fn GetGuardValue
  (return (Math|Float|MapRangeClamped (CallFunction|GetDataTrackerParameterValue 2) 0.0 1.0 0.0 1.0)))

(fn GetUnawareValue
  (return (Math|Float|MapRangeClamped (+ (CallFunction|GetDataTrackerParameterValue 0) (CallFunction|GetDataTrackerParameterValue 3)) 0.0 1.0 0.0 1.0)))

(fn OnDataTrackerUpdate
  (|CallParentFunction)
  (CallFunction|MIxState))
